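# Animation Blueprint for the F_MED_Shiny backpack, posing its wings to match the owning pawn's movement.
# A WingPosition state machine selects named poses for targeting, crouching, jumping, falling and nested skydive modes; Base adds a ground-movement blendspace.
# The Event Graph projects acceleration onto actor axes for that blendspace and remaps forward velocity into a sprint weight, zeroed during vertical motion.
# The cached result receives an additive Sprint pose except while crouching or skydiving.
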AnimBlueprint /BRCosmetics/Accessories/FORT_Backpacks/F_MED_Shiny_Backpack/Meshes/F_MED_Shiny_Pack_AnimBP
target_skeleton=/Game/Accessories/FORT_Backpacks/F_MED_Shiny_Backpack/Meshes/F_MED_Shiny_Pack_Skeleton
graphs [
 AnimGraph {nodes [
   n1 = Apply Additive(
    Base=Use cached pose 'preSprint'()
    Additive=Wing_Movement_Poses_Add_PoseAsset : Sprint!(PoseWeight:float=Get SprintSpeedRemap() -> (SprintSpeedRemap:double)) -> (Pose:Pose Link Structure) {
     is=title AnimGraphNode_PoseByName
     poseAsset=/BRCosmetics/Accessories/FORT_Backpacks/F_MED_Shiny_Backpack/Animation/Wing_Movement_Poses_Add_PoseAsset
     poseName=Sprint
    }
   ) {alphaBoolBlend {blendOption=Linear}}
   n2 = Use cached pose 'preSprint'()
   n3 = Get bIsCrouching()
   Output Pose(Result=Blend Poses by bool(
     BlendPose_0=Use cached pose 'preSprint'()
     BlendPose_1=Blend Poses by bool(BlendPose_0=@n2, BlendPose_1=@n1, BlendTime_0=0.3, BlendTime_1=0.3, bActiveValue=@n3) {blendType=Linear}
     BlendTime_0=0.3
     BlendTime_1=0.3
     bActiveValue=Get bIsSkyDiving()
    ) {blendType=Linear}) {layerGroup=None}
   AnimGraphNode_StateMachine_0 = WingPosition() {is=title AnimGraphNode_StateMachine}
   Save cached pose 'preSprint'(Pose=@AnimGraphNode_StateMachine_0) {cacheName=preSprint}
  ]}
 WingPosition {
  entry=Base
  path=AnimGraph.AnimGraphNode_StateMachine_0.WingPosition
  ids=id
  nodes [
   AnimStateNode_0 = Base() {is=title AnimStateNode}
   AnimStateNode_1 = Targeting() {is=title AnimStateNode}
   AnimStateNode_2 = Crouching() {is=title AnimStateNode}
   AnimStateNode_3 = Falling() {is=title AnimStateNode}
   AnimStateNode_4 = Jump() {is=title AnimStateNode}
   AnimStateNode_5 = SkydiveMode() {is=title AnimStateNode}
   AnimStateTransitionNode_0 = AnimStateTransitionNode()
   AnimStateTransitionNode_1 = AnimStateTransitionNode()
   AnimStateTransitionNode_10 = AnimStateTransitionNode()
   AnimStateTransitionNode_11 = AnimStateTransitionNode()
   AnimStateTransitionNode_2 = AnimStateTransitionNode()
   AnimStateTransitionNode_3 = AnimStateTransitionNode()
   AnimStateTransitionNode_4 = AnimStateTransitionNode()
   AnimStateTransitionNode_5 = AnimStateTransitionNode()
   AnimStateTransitionNode_6 = AnimStateTransitionNode()
   AnimStateTransitionNode_7 = AnimStateTransitionNode()
   AnimStateTransitionNode_8 = AnimStateTransitionNode()
   AnimStateTransitionNode_9 = AnimStateTransitionNode()
  ]
  links [
   {
    dst=Crouching
    rule=AnimGraph.AnimGraphNode_StateMachine_0.WingPosition.AnimStateTransitionNode_1.Transition
    src=Base
    via=AnimStateTransitionNode_1
   }
   {
    dst=Falling
    rule=AnimGraph.AnimGraphNode_StateMachine_0.WingPosition.AnimStateTransitionNode_3.Transition
    src=Base
    via=AnimStateTransitionNode_3
   }
   {
    dst=Jump
    rule=AnimGraph.AnimGraphNode_StateMachine_0.WingPosition.AnimStateTransitionNode_7.Transition
    src=Base
    via=AnimStateTransitionNode_7
   }
   {
    dst=SkydiveMode
    rule=AnimGraph.AnimGraphNode_StateMachine_0.WingPosition.AnimStateTransitionNode_9.Transition
    src=Base
    via=AnimStateTransitionNode_9
   }
   {
    dst=Targeting
    rule=AnimGraph.AnimGraphNode_StateMachine_0.WingPosition.AnimStateTransitionNode_0.Transition
    src=Base
    via=AnimStateTransitionNode_0
   }
   {
    dst=Base
    rule=AnimGraph.AnimGraphNode_StateMachine_0.WingPosition.AnimStateTransitionNode_6.Transition
    src=Crouching
    via=AnimStateTransitionNode_6
   }
   {
    dst=Base
    rule=AnimGraph.AnimGraphNode_StateMachine_0.WingPosition.AnimStateTransitionNode_4.Transition
    src=Falling
    via=AnimStateTransitionNode_4
   }
   {
    dst=Base
    rule=AnimGraph.AnimGraphNode_StateMachine_0.WingPosition.AnimStateTransitionNode_8.Transition
    src=Jump
    via=AnimStateTransitionNode_8
   }
   {
    dst=Falling
    rule=AnimGraph.AnimGraphNode_StateMachine_0.WingPosition.AnimStateTransitionNode_2.Transition
    src=Jump
    via=AnimStateTransitionNode_2
   }
   {
    dst=SkydiveMode
    rule=AnimGraph.AnimGraphNode_StateMachine_0.WingPosition.AnimStateTransitionNode_11.Transition
    src=Jump
    via=AnimStateTransitionNode_11
   }
   {
    dst=Base
    rule=AnimGraph.AnimGraphNode_StateMachine_0.WingPosition.AnimStateTransitionNode_10.Transition
    src=SkydiveMode
    via=AnimStateTransitionNode_10
   }
   {
    dst=Base
    rule=AnimGraph.AnimGraphNode_StateMachine_0.WingPosition.AnimStateTransitionNode_5.Transition
    src=Targeting
    via=AnimStateTransitionNode_5
   }
  ]
 }
 Base {
  path=AnimGraph.AnimGraphNode_StateMachine_0.WingPosition.AnimStateNode_0.Base
  nodes [
   n1 = Get PlayerVector_FwdBwd() -> (PlayerVector_FwdBwd:double)
   n2 = Get PlayerVector_LR() -> (PlayerVector_LR:double)
   Output Animation Pose(Result=Apply Additive(
     Base=Wing_Movement_Poses_PoseAsset : Base!(PoseWeight:float=1) -> (Pose:Pose Link Structure) {
      is=title AnimGraphNode_PoseByName
      poseAsset=/BRCosmetics/Accessories/FORT_Backpacks/F_MED_Shiny_Backpack/Animation/Wing_Movement_Poses_PoseAsset
      poseName=Base
     }
     Additive=Blendspace Player 'Ground_Movement_BS'(X=@n2, Y=@n1) {blendSpace=/BRCosmetics/Accessories/FORT_Backpacks/F_MED_Shiny_Backpack/Animation/Ground_Movement_BS}
    ) {alphaBoolBlend {blendOption=Linear}})
  ]
 }
 Targeting {
  path=AnimGraph.AnimGraphNode_StateMachine_0.WingPosition.AnimStateNode_1.Targeting
  nodes [Output Animation Pose(Result=Wing_Movement_Poses_PoseAsset : Targeting!(PoseWeight:float=1) -> (Pose:Pose Link Structure) {
     is=title AnimGraphNode_PoseByName
     poseAsset=&Wing_Movement_Poses_PoseAsset
     poseName=Targeting
    })]
 }
 Crouching {
  path=AnimGraph.AnimGraphNode_StateMachine_0.WingPosition.AnimStateNode_2.Crouching
  nodes [Output Animation Pose(Result=Wing_Movement_Poses_PoseAsset : Crouch!(PoseWeight:float=1) -> (Pose:Pose Link Structure) {
     is=title AnimGraphNode_PoseByName
     poseAsset=&Wing_Movement_Poses_PoseAsset
     poseName=Crouch
    })]
 }
 Falling {
  path=AnimGraph.AnimGraphNode_StateMachine_0.WingPosition.AnimStateNode_3.Falling
  nodes [Output Animation Pose(Result=Wing_Movement_Poses_PoseAsset : Jump!(PoseWeight:float=1) -> (Pose:Pose Link Structure) {
     is=title AnimGraphNode_PoseByName
     poseAsset=&Wing_Movement_Poses_PoseAsset
     poseName=Jump
    })]
 }
 Jump {
  path=AnimGraph.AnimGraphNode_StateMachine_0.WingPosition.AnimStateNode_4.Jump
  nodes [Output Animation Pose(Result=Wing_Movement_Poses_PoseAsset : Jump(PoseWeight=1) {poseAsset=&Wing_Movement_Poses_PoseAsset, poseName=Jump})]
 }
 SkydiveMode {
  path=AnimGraph.AnimGraphNode_StateMachine_0.WingPosition.AnimStateNode_5.SkydiveMode
  nodes [
   AnimGraphNode_StateMachine_0 = Skydiving State() {is=title AnimGraphNode_StateMachine}
   Output Animation Pose(Result=@AnimGraphNode_StateMachine_0)
  ]
 }
 Skydiving State {
  entry=Parachuting
  path=AnimGraph.AnimGraphNode_StateMachine_0.WingPosition.AnimStateNode_5.SkydiveMode.AnimGraphNode_StateMachine_0.Skydiving State
  ids=id
  nodes [
   AnimStateNode_0 = Parachuting() {is=title AnimStateNode}
   AnimStateNode_1 = Diving() {is=title AnimStateNode}
   AnimStateNode_2 = State_2() {is=title AnimStateNode}
   AnimStateTransitionNode_3 = AnimStateTransitionNode()
   AnimStateTransitionNode_4 = AnimStateTransitionNode()
   AnimStateTransitionNode_5 = AnimStateTransitionNode()
   AnimStateTransitionNode_6 = AnimStateTransitionNode()
   AnimStateTransitionNode_7 = AnimStateTransitionNode()
   AnimStateTransitionNode_8 = AnimStateTransitionNode()
  ]
  links [
   {
    dst=Parachuting
    rule=AnimGraph.AnimGraphNode_StateMachine_0.WingPosition.AnimStateNode_5.SkydiveMode.AnimGraphNode_StateMachine_0.Skydiving State.AnimStateTransitionNode_6.Transition
    src=Diving
    via=AnimStateTransitionNode_6
   }
   {
    dst=State_2
    rule=AnimGraph.AnimGraphNode_StateMachine_0.WingPosition.AnimStateNode_5.SkydiveMode.AnimGraphNode_StateMachine_0.Skydiving State.AnimStateTransitionNode_4.Transition
    src=Diving
    via=AnimStateTransitionNode_4
   }
   {
    dst=Diving
    rule=AnimGraph.AnimGraphNode_StateMachine_0.WingPosition.AnimStateNode_5.SkydiveMode.AnimGraphNode_StateMachine_0.Skydiving State.AnimStateTransitionNode_3.Transition
    src=Parachuting
    via=AnimStateTransitionNode_3
   }
   {
    dst=State_2
    rule=AnimGraph.AnimGraphNode_StateMachine_0.WingPosition.AnimStateNode_5.SkydiveMode.AnimGraphNode_StateMachine_0.Skydiving State.AnimStateTransitionNode_7.Transition
    src=Parachuting
    via=AnimStateTransitionNode_7
   }
   {
    dst=Diving
    rule=AnimGraph.AnimGraphNode_StateMachine_0.WingPosition.AnimStateNode_5.SkydiveMode.AnimGraphNode_StateMachine_0.Skydiving State.AnimStateTransitionNode_8.Transition
    src=State_2
    via=AnimStateTransitionNode_8
   }
   {
    dst=Parachuting
    rule=AnimGraph.AnimGraphNode_StateMachine_0.WingPosition.AnimStateNode_5.SkydiveMode.AnimGraphNode_StateMachine_0.Skydiving State.AnimStateTransitionNode_5.Transition
    src=State_2
    via=AnimStateTransitionNode_5
   }
  ]
 }
 Parachuting {
  path=AnimGraph.AnimGraphNode_StateMachine_0.WingPosition.AnimStateNode_5.SkydiveMode.AnimGraphNode_StateMachine_0.Skydiving State.AnimStateNode_0.Parachuting
  nodes [Output Animation Pose(Result=Wing_Movement_Poses_PoseAsset : Float!(PoseWeight:float=0.25) -> (Pose:Pose Link Structure) {
     is=title AnimGraphNode_PoseByName
     poseAsset=&Wing_Movement_Poses_PoseAsset
     poseName=Float
    })]
 }
 Diving {
  path=AnimGraph.AnimGraphNode_StateMachine_0.WingPosition.AnimStateNode_5.SkydiveMode.AnimGraphNode_StateMachine_0.Skydiving State.AnimStateNode_1.Diving
  nodes [Output Animation Pose(Result=Wing_Movement_Poses_PoseAsset : Dive!(PoseWeight:float=1) -> (Pose:Pose Link Structure) {
     is=title AnimGraphNode_PoseByName
     poseAsset=&Wing_Movement_Poses_PoseAsset
     poseName=Dive
    })]
 }
 State_2 {
  path=AnimGraph.AnimGraphNode_StateMachine_0.WingPosition.AnimStateNode_5.SkydiveMode.AnimGraphNode_StateMachine_0.Skydiving State.AnimStateNode_2.State_2
  nodes [Output Animation Pose(Result=Wing_Movement_Poses_PoseAsset : Float(PoseWeight=1) {poseAsset=&Wing_Movement_Poses_PoseAsset, poseName=Float})]
 }
 Transition {
  path=AnimGraph.AnimGraphNode_StateMachine_0.WingPosition.AnimStateNode_5.SkydiveMode.AnimGraphNode_StateMachine_0.Skydiving State.AnimStateTransitionNode_3.Transition
  nodes [Result(bCanEnterTransition=NOT Boolean(A=Get bIsParachuteOpened()))]
 }
 Transition {
  path=AnimGraph.AnimGraphNode_StateMachine_0.WingPosition.AnimStateNode_5.SkydiveMode.AnimGraphNode_StateMachine_0.Skydiving State.AnimStateTransitionNode_4.Transition
  nodes [Result(bCanEnterTransition=NOT Boolean(A=Get bIsSkydiveDiveMode()))]
 }
 Transition {
  path=AnimGraph.AnimGraphNode_StateMachine_0.WingPosition.AnimStateNode_5.SkydiveMode.AnimGraphNode_StateMachine_0.Skydiving State.AnimStateTransitionNode_5.Transition
  nodes [Result(bCanEnterTransition=Get bIsParachuteOpened())]
 }
 Transition {
  path=AnimGraph.AnimGraphNode_StateMachine_0.WingPosition.AnimStateNode_5.SkydiveMode.AnimGraphNode_StateMachine_0.Skydiving State.AnimStateTransitionNode_6.Transition
  nodes [Result(bCanEnterTransition=Get bIsParachuteOpened())]
 }
 Transition {
  path=AnimGraph.AnimGraphNode_StateMachine_0.WingPosition.AnimStateNode_5.SkydiveMode.AnimGraphNode_StateMachine_0.Skydiving State.AnimStateTransitionNode_7.Transition
  nodes [Result(bCanEnterTransition=NOT Boolean(A=Get bIsParachuteOpened()))]
 }
 Transition {
  path=AnimGraph.AnimGraphNode_StateMachine_0.WingPosition.AnimStateNode_5.SkydiveMode.AnimGraphNode_StateMachine_0.Skydiving State.AnimStateTransitionNode_8.Transition
  nodes [Result(bCanEnterTransition=Get bIsSkydiveDiveMode())]
 }
 Transition {
  path=AnimGraph.AnimGraphNode_StateMachine_0.WingPosition.AnimStateTransitionNode_0.Transition
  nodes [Result(bCanEnterTransition=Get bIsTargeting())]
 }
 Transition {
  path=AnimGraph.AnimGraphNode_StateMachine_0.WingPosition.AnimStateTransitionNode_1.Transition
  nodes [Result(bCanEnterTransition=Get bIsCrouching())]
 }
 Transition {
  path=AnimGraph.AnimGraphNode_StateMachine_0.WingPosition.AnimStateTransitionNode_10.Transition
  nodes [Result(bCanEnterTransition=NOT Boolean(A=Get bIsSkyDiving()))]
 }
 Transition {
  path=AnimGraph.AnimGraphNode_StateMachine_0.WingPosition.AnimStateTransitionNode_11.Transition
  nodes [Result(bCanEnterTransition=Get bIsSkyDiving())]
 }
 Transition {
  path=AnimGraph.AnimGraphNode_StateMachine_0.WingPosition.AnimStateTransitionNode_2.Transition
  nodes [Result(bCanEnterTransition=Get bIsFalling())]
 }
 Transition {
  path=AnimGraph.AnimGraphNode_StateMachine_0.WingPosition.AnimStateTransitionNode_3.Transition
  nodes [Result(bCanEnterTransition=Get bIsFalling())]
 }
 Transition {
  path=AnimGraph.AnimGraphNode_StateMachine_0.WingPosition.AnimStateTransitionNode_4.Transition
  nodes [Result(bCanEnterTransition=NOT Boolean(A=Get bIsFalling()))]
 }
 Transition {
  path=AnimGraph.AnimGraphNode_StateMachine_0.WingPosition.AnimStateTransitionNode_5.Transition
  nodes [Result(bCanEnterTransition=NOT Boolean(A=Get bIsTargeting()))]
 }
 Transition {
  path=AnimGraph.AnimGraphNode_StateMachine_0.WingPosition.AnimStateTransitionNode_6.Transition
  nodes [Result(bCanEnterTransition=NOT Boolean(A=Get bIsCrouching()))]
 }
 Transition {
  path=AnimGraph.AnimGraphNode_StateMachine_0.WingPosition.AnimStateTransitionNode_7.Transition
  nodes [Result(bCanEnterTransition=Get bIsJumping())]
 }
 Transition {
  path=AnimGraph.AnimGraphNode_StateMachine_0.WingPosition.AnimStateTransitionNode_8.Transition
  nodes [Result(bCanEnterTransition=NOT Boolean(A=Get bIsJumping()))]
 }
 Transition {
  path=AnimGraph.AnimGraphNode_StateMachine_0.WingPosition.AnimStateTransitionNode_9.Transition
  nodes [Result(bCanEnterTransition=Get bIsSkyDiving())]
 }
 EventGraph {nodes [
   Event BlueprintUpdateAnimation()
   n1 = TryGetPawnOwner()
   > n2 = Cast To Character!(execute, Object=@n1) -> (then, CastFailed, AsCharacter:Character Object Reference) {is=title K2Node_DynamicCast}
   n3 = Get Actor Rotation!(self:Actor Object Reference=@n1) -> (ReturnValue:Rotator) {is=title K2Node_CallFunction}
   n4 = GetCurrentAcceleration!(self:Character Movement Component Object Reference=Get CharacterMovement!(self:Character Object Reference=@n2) -> (CharacterMovement:Character Movement Component Object Reference)) -> (ReturnValue:Vector)
   n5 = BreakVector(InVec=vector * vector(A:Vector=Get Rotation X Vector!(InRot:Rotator=@n3) -> (ReturnValue:Vector) {is=title K2Node_CallFunction}, B:Vector=@n4) -> (ReturnValue:Vector) {is=title K2Node_CommutativeAssociativeBinaryOperator})
   > Set PlayerVector_FwdBwd(PlayerVector_FwdBwd:double=float + float(A=@n5, B=@n5.Y) {is=title K2Node_CommutativeAssociativeBinaryOperator}) -> (Output_Get:double)
   n6 = BreakVector(InVec=vector * vector(A=GetAxes!(A:Rotator=@n3) -> (X:Vector, Y:Vector, Z:Vector).Y, B=@n4))
   > Set PlayerVector_LR(PlayerVector_LR:double=float + float(A=@n6, B=@n6.Y)) -> (Output_Get:double)
   n7 = TryGetPawnOwner()
   n8 = GetVelocity!(self:Actor Object Reference=@n7) -> (ReturnValue:Vector)
   n9 = "Not Equal (Float)"!(A:double=BreakVector(InVec=@n8).Z, B:double=0) -> (ReturnValue:bool) {is=title K2Node_CallFunction}
   n10 = Get Actor Rotation(self=@n7)
   n11 = BreakVector(InVec=vector * vector(A=Get Rotation X Vector(InRot=@n10), B=@n8))
   n12 = float + float(A=@n11, B=@n11.Y)
   > Set SprintSpeedRemap(SprintSpeedRemap:double=float * float(A=MapRangeClamped(Value=@n12, InRangeA=0, InRangeB=550), B=SelectFloat(A=0, B=1, bPickA=@n9))) -> (Output_Get:double)
  ]}
]
variables [
 PlayerVector_FwdBwd {type=number}
 PlayerVector_LR {type=number}
 SprintSpeedRemap {type=number}
]
